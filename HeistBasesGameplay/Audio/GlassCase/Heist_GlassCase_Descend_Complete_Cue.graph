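# Sound cue for the completion of a glass case descending in the Heist Bases gameplay content.
# A Random node picks one of three equally weighted completion wave variants and feeds it into a Modulator at the root.
# The cue uses the glass case descend attenuation asset, the gameplay events sound class and a slightly raised volume.
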SoundCue /HeistBasesGameplay/Audio/GlassCase/Heist_GlassCase_Descend_Complete_Cue
root=Modulator_0
nodes [
 {id=Modulator_0, children [Random_1]}
 {id=Random_1, children [WavePlayer_3,WavePlayer_4,WavePlayer_5], properties {weights [1,1,1]}}
 {id=WavePlayer_3, properties {soundWaveAssetPtr=&/Heist_GlassCase_Descend_Complete_03}}
 {id=WavePlayer_4, properties {soundWaveAssetPtr=&/Heist_GlassCase_Descend_Complete_01}}
 {id=WavePlayer_5, properties {soundWaveAssetPtr=&/Heist_GlassCase_Descend_Complete_02}}
]
properties {
 attenuationSettings=/HeistBasesGameplay/Audio/GlassCase/Settings/GlassCase_Descend
 maxDistance=5200
 soundClassObject=&Fort_Events_Gameplay
 volumeMultiplier=1.2
}
