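# Sound cue for a glass case descending in the Heist Bases gameplay content.
# A Random node picks one of three descend wave variations with equal weight and is excluded from branch culling.
# It plays through the Fort_Events_Gameplay sound class with GlassCase_Descend attenuation and slightly reduced volume.
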SoundCue /HeistBasesGameplay/Audio/GlassCase/Heist_GlassCase_Descend_Cue
root=Random_0
nodes [
 {id=Random_0, children [WavePlayer_0,WavePlayer_1,WavePlayer_2], properties {bShouldExcludeFromBranchCulling=true, weights [1,1,1]}}
 {id=WavePlayer_0, properties {soundWaveAssetPtr=&/Heist_GlassCase_Descend_03}}
 {id=WavePlayer_1, properties {soundWaveAssetPtr=&/Heist_GlassCase_Descend_01}}
 {id=WavePlayer_2, properties {soundWaveAssetPtr=&/Heist_GlassCase_Descend_02}}
]
properties {
 attenuationSettings=/HeistBasesGameplay/Audio/GlassCase/Settings/GlassCase_Descend
 maxDistance=5200
 soundClassObject=&Fort_Events_Gameplay
 volumeMultiplier=0.8
}
