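# Sound cue for punching a glass case in Heist bases gameplay.
# A random node picks one of five equally weighted punch wave variations, excluded from branch culling, then feeds a modulator for subtle variation.
# It plays through the first-person player foley sound class with basic foley attenuation and a boosted cue volume.
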SoundCue /HeistBasesGameplay/Audio/GlassCase/Heist_GlassCase_Punch_Cue
root=Modulator_0
nodes [
 {id=Modulator_0, children [Random_0]}
 {
  id=Random_0
  children [WavePlayer_0,WavePlayer_1,WavePlayer_2,WavePlayer_3,WavePlayer_4]
  properties {bShouldExcludeFromBranchCulling=true, weights [1,1,1,1,1]}
 }
 {id=WavePlayer_0, properties {soundWaveAssetPtr=&/Heist_GlassCase_Punch_01}}
 {id=WavePlayer_1, properties {soundWaveAssetPtr=&/Heist_GlassCase_Punch_02}}
 {id=WavePlayer_2, properties {soundWaveAssetPtr=&/Heist_GlassCase_Punch_03}}
 {id=WavePlayer_3, properties {soundWaveAssetPtr=&/Heist_GlassCase_Punch_04}}
 {id=WavePlayer_4, properties {soundWaveAssetPtr=&/Heist_GlassCase_Punch_05}}
]
properties {
 attenuationSettings=&Foley_Attenuation_Basic_Att
 maxDistance=1536
 soundClassObject=&Fort_Player_Foley_1P_Athena
 volumeMultiplier=2
}
