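# Sound cue for a heist glass case being smashed.
# A Random node picks one of five equally weighted smash waves and feeds a Modulator at the root for pitch and volume variation.
# It borrows a medium-object destruction attenuation, routes to the gameplay music sound class with a slight volume boost, and its concurrency override blocks new instances at the limit.
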SoundCue /HeistBasesGameplay/Audio/GlassCase/Heist_GlassCase_Smashed_Cue
root=Modulator_0
nodes [
 {id=Modulator_0, children [Random_0]}
 {id=Random_0, children [WavePlayer_0,WavePlayer_1,WavePlayer_2,WavePlayer_3,WavePlayer_4], properties {weights [1,1,1,1,1]}}
 {id=WavePlayer_0, properties {soundWaveAssetPtr=&/Heist_GlassCase_Smash_01}}
 {id=WavePlayer_1, properties {soundWaveAssetPtr=&/Heist_GlassCase_Smash_02}}
 {id=WavePlayer_2, properties {soundWaveAssetPtr=&/Heist_GlassCase_Smash_03}}
 {id=WavePlayer_3, properties {soundWaveAssetPtr=&/Heist_GlassCase_Smash_04}}
 {id=WavePlayer_4, properties {soundWaveAssetPtr=&/Heist_GlassCase_Smash_05}}
]
properties {
 attenuationSettings=/Game/Sounds/Attenuation/Fort_Demolition_Destroy_Medium_Sized_Object_Att
 maxDistance=4000
 soundClassObject=/Game/Sounds/FortSoundClassesAndModes/Fort_Music_Gameplay
 volumeMultiplier=1.17
 concurrencyOverrides {resolutionRule=PreventNew}
}
